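# UI material for the ToyFox mobile icon backplate, drawing a red hexagon with alpha-composite blending.
# A quarter-turn-rotated polygon SDF feeds an anti-aliased inner mask and a thin outline mask, so the shape needs no textures.
# Two reds blend along a rotated UV gradient inside the fill, then mix with a dark outline-masked tone using the outline plus a weighted gradient.
# A separate lerp combines the fill and outline masks.
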Material /ToyFoxGameplay/UI/Mobile/M_UI_ToyFox_IconBackplate
graph {
 properties {bAutomaticallySetUsageInEditor=false, blendMode=BLEND_AlphaComposite, materialDomain=MD_UI, usedShadingModels=Default Lit}
 nodes [
  n1 = MaterialFunctionCall() {materialFunction=&GetUserInterfaceUV}
  n2 = MaterialFunctionCall!(
   UVs=MaterialFunctionCall(UVs=@n1.Normalized UV, "Rotation Angle (0-1)"=Constant() {r=0.25}) {materialFunction=&CustomRotator}.Rotated Values
   Radius=Constant() {r=1}
   Sides=Constant() {r=6}
   InnerAngle
  ) {materialFunction=/FortUILibrary/MaterialFunctions/Shapes/SDFs/MF_SDF_RegularPolygon}
  n3 = MaterialFunctionCall!(DistanceField=@n2.SDF, Threshold, Thickness=Constant() {r=0.025}, Alignment=Constant() {r=-3}, Smoothing) {materialFunction=/FortUILibrary/MaterialFunctions/Shapes/Modifiers/MF_UI_ShapeOutline}
  n4 = Constant3Vector() {constant=<rgba 0.6867 0.0369 0.0452 1>}
  n5 = MaterialFunctionCall(DistanceField=@n2.SDF) {materialFunction=&MF_UI_AntiAliasing}
  n6 = Constant3Vector() {constant=<rgba 0.1683 0.0252 0.0262 1>}
  n7 = MaterialFunctionCall(UVs=@n1.Normalized UV, "Rotation Angle (0-1)"=Constant() {r=0.1}) {materialFunction=&CustomRotator}
  n8 = NamedRerouteUsage()
  LinearInterpolate(
   A=Multiply(A=Constant3Vector() {constant=<rgba 0.0296 0.0194 0.0194 1>}, B=@n3.Outline)
   B=LinearInterpolate(A=Multiply(A=@n4, B=@n5.Inner), B=Multiply(A=@n6, B=@n5.Inner), Alpha=ComponentMask(None=@n7.Rotated Values) {r=true})
   Alpha=Add(A=@n3.Outline, B=Multiply(A=@n8) {constB=0.25})
  )
  NamedRerouteDeclaration(None=ComponentMask(None=TextureCoordinate()) {g=true}) {name=V Gradient, nodeColor=<rgba 0 0 0 1>}
  Reroute(None=LinearInterpolate(A=@n5.Inner, B=@n3.Outline) {constAlpha=0.15})
 ]
}
